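; Puppeteer boss's floating mask actor, linked by Niagara strings to the demon pawn's hands and facing its aggro target via server timer and client tick.
; Demon phase 2 enables eye weakpoints and randomized, hotfix-tuned teleports to EQS-picked points, followed by fireball ability events.
; Broken eyes stagger mask and demon, pausing teleports and look-at.
; Body damage is instantly healed back; demon death dissolves and destroys the mask.
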(event Custom|InitiatedChallengeStart
  ;[0.05373 0.06296 0.15 0.5] BeginPlay
  (CallFunction|ClientInitiatedChallengeStart))

(event Custom|CE_MaskInPhase2ToEnableWeakpoints
  ;[0.05373 0.06296 0.15 0.5] BeginPlay
  (Networking|CanExecuteCosmeticEvents
    (:True
      (Effects|Components|ParticleSystem|SetBoolParameter (Variables|Default|GetMaskIdleStringFX) "EnableWeakpoints" true))))

(event EventBeginPlay
  (bind _maskidlestringfx (Variables|Default|GetMaskIdleStringFX))
  ;[0.05373 0.06296 0.15 0.5] BeginPlay
  (Networking|SwitchHasAuthority
    (:Authority
      (Variables|RotationControl|SetMaskRotation (Transformation|GetActorRotation))
      (Variables|LocationControl|SetMaskLocation (Transformation|GetActorLocation))
      (Class|FortAsyncActionCurrentPlaylistReady|PlaylistDataReadyAsync
        (:OnReady
          (Utilities|FlowControl|Delay 0.5)
          (bind _asbp_player_pawn_puppeteer_demon (Utilities|Casting|CastToBP_PlayerPawn_Puppeteer_Demon (Game|GetInstigator)))
          (bind _output_get (Variables|Default|SetDemonPawn _asbp_player_pawn_puppeteer_demon))
          (Health|BindEventtoOnPawnHealthChanged _output_get (EventDispatchers|CreateEvent :!event Custom|DemonHealthChanged))
          (EventDispatchers|BindEventtoOnDied _output_get (EventDispatchers|CreateEvent :!event Custom|DemonPawnDied) :!class /Script/FortniteGame.FortPawn)
          (Default|BindEventtoEDInitiatedChallengeStart (Variables|Default|GetDemonPawn) (delegate Custom|InitiatedChallengeStart))
          (bind _asbga_puppeteer_mask_eye_weakpoint (Utilities|Casting|CastToBGA_PuppeteerMaskEyeWeakpoint (Variables|/Script/Engine.ChildActorComponent|GetChildActor (Variables|Default|GetEyeWeakpointL))))
          (Variables|EyesControl|SetLeftEyeActor _asbga_puppeteer_mask_eye_weakpoint)
          (bind _asbga_puppeteer_mask_eye_weakpoint_2 (Utilities|Casting|CastToBGA_PuppeteerMaskEyeWeakpoint (Variables|/Script/Engine.ChildActorComponent|GetChildActor (Variables|Default|GetEyeWeakpointR))))
          (Variables|EyesControl|SetRightEyeActor _asbga_puppeteer_mask_eye_weakpoint_2)
          (bind _returnvalue (Utilities|Time|SetTimerbyEvent (EventDispatchers|CreateEvent :!event Custom|LookAtTarget) (Variables|RotationControl|GetServerLookAtTimeInterval) true true))
          (Variables|Default|SetLookAtTargetTimerHandle _returnvalue)
          (CallFunction|StartMaskOscillation)
          (Variables|EyesControl|SetEyesIgnoreWeaponCollision true)
          (Class|BPFLAICommonFunctions|AddUniqueGameplayTag self "(TagName=\"PuppeteerBoss.Mask.Ability.Teleport.OnCooldown\")")
          (Ability|Tasks|WaitGameplayTagAddToActor (Variables|Default|GetDemonPawn) "(TagName=\"PuppeteerBoss.Demon.IsInPhase2\")" true
            (:Added
              (Class|BPFLAICommonFunctions|AddUniqueGameplayTag self "(TagName=\"PuppeteerBoss.Demon.IsInPhase2\")")
              (Variables|EyesControl|SetEyesIgnoreWeaponCollision false)
              (GameplayCue|AddGameplayCueOnActor(Looping) self (Variables|FX|GetGC_MaskAggressive_Loop) (Ability|GameplayCue|MakeGameplayCueParameters 0.0))
              (CallFunction|ToggleEyeGlow false true true)
              (bind _returnvalue_2 (Utilities|Time|SetTimerbyEvent (EventDispatchers|CreateEvent :!event Custom|InitiateTeleport) (Math|Float|Clamp(Float) (- (Math|Random|RandomFloatinRange (ScalableFloat|GetValueatLevel (Variables|Hotfix|GetTeleportCdMin_Hf)) (ScalableFloat|GetValueatLevel (Variables|Hotfix|GetTeleportCdMax_Hf))) (ScalableFloat|GetValueatLevel (Variables|Hotfix|GetTeleportTargetSwitchInvertDelay_Hf))) 0.0 (ScalableFloat|GetValueatLevel (Variables|Hotfix|GetTeleportCdMin_Hf))) false true))
              (Variables|Default|SetInitiateTeleportTimerHandle _returnvalue_2)
              (CallFunction|CEMaskinPhase2ToEnableWeakpoints)
              (Utilities|FlowControl|Delay 3.0)
              (Variables|Default|SetProxTeleportLocked false))))))
    (:Remote
      (Class|FortAsyncActionCurrentPlaylistReady|PlaylistDataReadyAsync
        (:OnReady
          (Utilities|FlowControl|Sequence
            (:then_0
              (Effects|Components|ParticleSystem|SetBoolParameter (Variables|Default|GetMaskIdleStringFX) "EnableWeakpoints" false))
            (:then_1
              (Transformation|SetActorRotation (Variables|RotationControl|GetMaskRotation))
              (Transformation|SetActorLocation (Variables|LocationControl|GetMaskLocation) false true))
            (:then_2
              (Effects|Components|ParticleSystem|SetBoolParameter _maskidlestringfx "HideLeftEyeFX" (Variables|EyesControl|GetLeftEyeBroken))
              (Effects|Components|ParticleSystem|SetBoolParameter _maskidlestringfx "HideRightEyeFX" (Variables|EyesControl|GetRightEyeBroken)))))))))

(event Custom|ClientInitiatedChallengeStart
  ;[0.05373 0.06296 0.15 0.5] BeginPlay
  (Networking|CanExecuteCosmeticEvents
    (:True
      (Effects|Components|ParticleSystem|SetBoolParameter (Variables|Default|GetMaskIdleStringFX) "EnableWeakpoints" (Variables|/PuppeteerBoss/Pawn/BP_PlayerPawn_Puppeteer_Demon.BP_PlayerPawn_Puppeteer_Demon_C|GetIsStartPhase2 (Variables|Default|GetDemonPawn))))))

(event Custom|RestartTeleportCooldown
  ;[0.05373 0.06296 0.15 0.5] BeginPlay
  (bind _returnvalue (Utilities|Time|SetTimerbyEvent (EventDispatchers|CreateEvent :!event Custom|InitiateTeleport) (Math|Float|Clamp(Float) (- (Math|Random|RandomFloatinRange (ScalableFloat|GetValueatLevel (Variables|Hotfix|GetTeleportCdMin_Hf)) (ScalableFloat|GetValueatLevel (Variables|Hotfix|GetTeleportCdMax_Hf))) (ScalableFloat|GetValueatLevel (Variables|Hotfix|GetTeleportTargetSwitchInvertDelay_Hf))) 0.0 (ScalableFloat|GetValueatLevel (Variables|Hotfix|GetTeleportCdMin_Hf))) false true))
  (Variables|Default|SetInitiateTeleportTimerHandle _returnvalue)
  (CallFunction|CEMaskinPhase2ToEnableWeakpoints)
  (Utilities|FlowControl|Delay 3.0)
  (Variables|Default|SetProxTeleportLocked false))

(event Custom|ClientRoseDepthStartPhase1
  ;[0.05373 0.06296 0.15 0.5] BeginPlay
  ;; Start Ticking the string demon target location
  (Utilities|Time|ClearandInvalidateTimerbyHandle (Variables|Default|GetUnawareStateTickTimerHandler))
  (bind _returnvalue (Utilities|Time|SetTimerbyEvent (delegate Custom|TickDemonPawnLocation) 0.1 true))
  (Variables|Default|SetTickDemonPawnLocationTimerHandle _returnvalue))

(event Custom|DemonPawnSet
  (params (DemonPawn :type BP%20Player%20Pawn%20Puppeteer%20Demon%20Object%20Reference))
  ;[0.05373 0.06296 0.15 0.5] BeginPlay
  (Utilities|IsValid DemonPawn
    (:Is%20Valid
      (Variables|Default|SetDemonPawnClient DemonPawn)
      (Effects|Components|ParticleSystem|SetBoolParameter (Variables|Default|GetMaskIdleStringFX) "EnableStrings" true)
      (CallFunction|SetStringFXTarget)
      (bind _returnvalue (Utilities|Time|SetTimerbyEvent (delegate Custom|UnawareStateTick) 0.2 true))
      (Variables|Default|SetUnawareStateTickTimerHandler _returnvalue)
      (EventDispatchers|BindEventtoOnDBNOStateChanged (Variables|Default|GetDemonPawnClient) (EventDispatchers|CreateEvent :!event Custom|ActivateAudioEffect) :!class /Script/FortniteGame.FortPawn)
      (Default|BindEventtoRoseDepthStartPhase1 (Variables|Default|GetDemonPawnClient) (delegate Custom|ClientRoseDepthStartPhase1)))
    (:Is%20Not%20Valid
      (Development|PrintString "ERROR:: BGA_PuppeteerMask:: DemonPawn Ref isn't valid" true true "(R=1.000000,G=0.021699,B=0.000000,A=1.000000)" 999.0))))

(event Custom|TickDemonPawnLocation
  ;[0.05373 0.06296 0.15 0.5] BeginPlay
  ;; Start Ticking the string demon target location
  (CallFunction|TickDemonPawnLocationonClient (Transformation|GetSocketLocation (Variables|/Script/Engine.Character|GetMesh (Variables|Default|GetDemonPawn)) "hand_l") (Transformation|GetSocketLocation (Variables|/Script/Engine.Character|GetMesh (Variables|Default|GetDemonPawn)) "hand_r")))

(event Custom|UnawareStateTick
  ;[0.05373 0.06296 0.15 0.5] BeginPlay
  (CallFunction|TickDemonPawnLocationonClient (Transformation|GetSocketLocation (Variables|/Script/Engine.Character|GetMesh (Variables|Default|GetDemonPawn)) "hand_l") (Transformation|GetSocketLocation (Variables|/Script/Engine.Character|GetMesh (Variables|Default|GetDemonPawn)) "hand_r")))

(event Custom|UpdateStringFXonMask
  (params (PuppetLocationLeftArm :type Vector) (PuppetLocationRightArm :type Vector))
  (bind _maskidlestringfx (Variables|Default|GetMaskIdleStringFX))
  ;[0.9427 0.04419 0.7077 0.5] Mask String FX - Timer update
  (Networking|CanExecuteCosmeticEvents
    (:True
      (Effects|Components|ParticleSystem|SetVectorParameter _maskidlestringfx "StringTargetL" PuppetLocationLeftArm)
      (Effects|Components|ParticleSystem|SetVectorParameter _maskidlestringfx "StringTargetR" PuppetLocationRightArm)
      (Components|Activation|Deactivate _maskidlestringfx)
      (Utilities|FlowControl|DelayUntilNextTick)
      (if (not (Variables|Default|GetIsDeath))
        (Components|Activation|Activate _maskidlestringfx true)))))

(event Custom|ToggleEyeGlow
  (params (bIsLeft :type Boolean) (bIsBoth :type Boolean) (Activate :type Boolean))
  ; Mask Weakpoint Setup
  (Networking|CanExecuteCosmeticEvents
    (:True
      (if bIsBoth
        (Rendering|Material|SetCustomPrimitiveDataVector2 (Variables|Default|GetSkeletalMeshComponentBudgeted) 3 (Math|Vector2D|MakeVector2D (Math|Float|SelectFloat 1.0 0.0 Activate) (Math|Float|SelectFloat 1.0 0.0 Activate))) ; Enable/disable glow on both eyes\nThis is an option to avoid sending multiple multicast events
        (else
          (Rendering|Material|SetCustomPrimitiveDataFloat (Variables|Default|GetSkeletalMeshComponentBudgeted) (Math|Integer|SelectInt 4 3 bIsLeft) (Math|Float|SelectFloat 1.0 0.0 Activate)) ; Enable/disable glow on specified eye
        )))))

(event Custom|TickDemonPawnLocationOnClient
  (params (PuppetLocationLeftArm :type Vector) (PuppetLocationRightArm :type Vector))
  (bind _maskidlestringfx (Variables|Default|GetMaskIdleStringFX))
  ;[0.9427 0.04419 0.7077 0.5] Mask String FX - Timer update
  (Networking|CanExecuteCosmeticEvents
    (:True
      (Effects|Components|ParticleSystem|SetVectorParameter _maskidlestringfx "StringTargetL" PuppetLocationLeftArm)
      (Effects|Components|ParticleSystem|SetVectorParameter _maskidlestringfx "StringTargetR" PuppetLocationRightArm))))

(event EventTick
  (params (DeltaSeconds :type Float%20%28single-precision%29))
  ; Tick - used to rotate actor on client, server is only ticked for debug.
  (Networking|SwitchHasAuthority
    (:Authority
      _)
    (:Remote
      (CallFunction|LookatTarget)))
  (if (Variables|RotationControl|GetDebugShowRotations)
    (Rendering|Debug|DrawDebugArrow (Transformation|GetActorLocation) (+ (Transformation|GetActorLocation) (* (Transformation|GetActorForwardVector) "600.000000,600.000000,600.000000")) 4e+01 (Math|Color|SelectColor "(R=1.000000,G=0.000000,B=0.000000,A=1.000000)" "(R=0.000000,G=1.000000,B=0.000000,A=1.000000)" (Networking|HasAuthority)) 0.0 6.0)))

(event Custom|LookAtTarget
  (bind _returnvalue (Networking|HasAuthority))
  ;[0.13 0.04624 0.15 0.5] LookAtTarget - called every tick for client, on time intervals for server
  (if (Variables|FunctionControl|GetPauseLookAtTarget)
    _
    (elif _returnvalue
      (bind _targetactor (Class|BPFLAICommonFunctions|GetTargetActorfromBlackboard (Variables|Default|GetDemonPawn)))
      (bind _output_get (Variables|Default|SetPlayerTarget _targetactor))
      (Utilities|IsValid _output_get
        (:Is%20Valid
          (if (Health|IsDBNO _output_get)
            (Class|BPPlayerPawnPuppeteerDemon|TrySwitchAggroTarget (Variables|Default|GetDemonPawn) :!label _label))
          (bind _rotation (CallFunction|FindandInterpTargetLookatRotation false (Math|Float|SelectFloat (Variables|RotationControl|GetServerLookAtTimeInterval) (Utilities|Time|GetWorldDeltaSeconds) _returnvalue) :!label _label_2))
          (bind _output_get_2 (Variables|RotationControl|SetMaskRotation (Math|Rotator|MakeRotator (.roll _rotation) (Math|Float|ClampAngle (.pitch _rotation) -4e+01 4e+01) (.yaw _rotation))))
          (bind (_output_get_roll _output_get_pitch _output_get_yaw) (Math|Rotator|BreakRotator _output_get_2))
          (Transformation|SetActorRotation 0 (Math|Rotator|MakeRotator 0.0 0.0 _output_get_yaw) false))
        (:Is%20Not%20Valid
          _))
      (goto _label)))
  (goto _label_2))

(event Custom|EvaluateDistanceToPlayerTarget
  ;[0.07615 0.15 0.0667 0.5] EvaluateDistanceToPlayerTarget
  (Utilities|IsValid (Variables|Default|GetPlayerTarget)
    (:Is%20Valid
      (if (Math|Boolean|ANDBoolean (<= (Math|Vector|Distance2D(Vector) (Transformation|GetActorLocation) (Transformation|GetActorLocation (Variables|Default|GetPlayerTarget))) (ScalableFloat|GetValueatLevel (Variables|Hotfix|GetPlayerProximityTeleportDist_Hf))) (not (GameplayTags|HasAnyMatchingGameplayTags (GameplayTags|MakeLiteralGameplayTagContainer "(GameplayTags=((TagName=\"PuppeteerBoss.Mask.Ability.Teleport.Activate\")))"))) (not (Variables|Default|GetProxTeleportLocked)) (ScalableFloat|GetValueasBool (Variables|Hotfix|GetTeleportEnabled_Hf)))
        (CallFunction|InitiateTeleport)))))

(event Custom|InitiateTeleport
  ;[0.15 0.06635 0.02993 0.5] TeleportEvents
  (if (Math|Random|RandomBoolwithWeight (ScalableFloat|GetValueatLevel (Variables|Hotfix|GetTargetSwitchBeforeTeleportChance_Hf)))
    (Class|BPPlayerPawnPuppeteerDemon|TrySwitchAggroTarget (Variables|Default|GetDemonPawn)))
  (Utilities|FlowControl|Delay (Math|Float|Clamp(Float) (ScalableFloat|GetValueatLevel (Variables|Hotfix|GetTeleportTargetSwitchInvertDelay_Hf)) 0.0 (ScalableFloat|GetValueatLevel (Variables|Hotfix|GetTeleportCdMin_Hf))))
  (if (ScalableFloat|GetValueasBool (Variables|Hotfix|GetTeleportEnabled_Hf))
    (if (not (Variables|FunctionControl|GetTeleportAreaBoxValidated))
      (bind _validated (CallFunction|ValidateTeleportAreaBox))
      (if (not _validated)
        (CallFunction|RestartTeleportCooldown)))
    (if (GameplayTags|HasMatchingGameplayTag "(TagName=\"FirePetalBosses.Demons.IsinSpawnCloud\")")
      _
      (elif (Utilities|Time|IsValidTimerHandle (Variables|Default|GetTickDemonPawnLocationTimerHandle))
        _))
    (if (not (Variables|FunctionControl|GetPauseTeleport))
      (Utilities|IsValid (Variables|Default|GetPlayerTarget)
        (:Is%20Valid
          (Class|BPFLAICommonFunctions|AddUniqueGameplayTag self "(TagName=\"PuppeteerBoss.Mask.Ability.Teleport.Activate\")")
          (Utilities|Time|ClearandInvalidateTimerbyHandle (Variables|Default|GetShootFireballsTimerHandle))
          (bind _returnvalue (AI|EQS|RunEQSQuery "/PuppeteerBoss/EQS/EQS_MaskTeleportLocation.EQS_MaskTeleportLocation" self "RandomBest5Pct"))
          (AI|EQS|BindEventtoOnQueryFinished _returnvalue (delegate Custom|TeleportPointQueryFinished)))
        (:Is%20Not%20Valid
          _)))
    (CallFunction|RestartTeleportCooldown)
    (else
      (Networking|SwitchHasAuthority
        (:Authority
          (bind _returnvalue_2 (Utilities|Time|SetTimerbyEvent (EventDispatchers|CreateEvent :!event Custom|InitiateFireballsAbility) (ScalableFloat|GetValueatLevel (Variables|Hotfix|GetInitiateFireballsDelay_Hf))))
          (Variables|Default|SetShootFireballsTimerHandle _returnvalue_2)
          (Utilities|FlowControl|Delay (ScalableFloat|GetValueatLevel (Variables|Hotfix|GetTeleportGracePeriod_Hf)))
          (CallFunction|RestartTeleportCooldown))))))

(event Custom|Phase2ResetInitiateTeleport
  ;[0.15 0.06635 0.02993 0.5] TeleportEvents
  (Utilities|FlowControl|Delay (ScalableFloat|GetValueatLevel (Variables|Hotfix|GetPhase2InitiateTeleportDelay_Hf)))
  (if (Math|Random|RandomBoolwithWeight (ScalableFloat|GetValueatLevel (Variables|Hotfix|GetTargetSwitchBeforeTeleportChance_Hf)))
    (Class|BPPlayerPawnPuppeteerDemon|TrySwitchAggroTarget (Variables|Default|GetDemonPawn)))
  (Utilities|FlowControl|Delay (Math|Float|Clamp(Float) (ScalableFloat|GetValueatLevel (Variables|Hotfix|GetTeleportTargetSwitchInvertDelay_Hf)) 0.0 (ScalableFloat|GetValueatLevel (Variables|Hotfix|GetTeleportCdMin_Hf))))
  (if (ScalableFloat|GetValueasBool (Variables|Hotfix|GetTeleportEnabled_Hf))
    (if (not (Variables|FunctionControl|GetTeleportAreaBoxValidated))
      (bind _validated (CallFunction|ValidateTeleportAreaBox))
      (if (not _validated)
        (CallFunction|RestartTeleportCooldown)))
    (if (GameplayTags|HasMatchingGameplayTag "(TagName=\"FirePetalBosses.Demons.IsinSpawnCloud\")")
      _
      (elif (Utilities|Time|IsValidTimerHandle (Variables|Default|GetTickDemonPawnLocationTimerHandle))
        _))
    (if (not (Variables|FunctionControl|GetPauseTeleport))
      (Utilities|IsValid (Variables|Default|GetPlayerTarget)
        (:Is%20Valid
          (Class|BPFLAICommonFunctions|AddUniqueGameplayTag self "(TagName=\"PuppeteerBoss.Mask.Ability.Teleport.Activate\")")
          (Utilities|Time|ClearandInvalidateTimerbyHandle (Variables|Default|GetShootFireballsTimerHandle))
          (bind _returnvalue (AI|EQS|RunEQSQuery "/PuppeteerBoss/EQS/EQS_MaskTeleportLocation.EQS_MaskTeleportLocation" self "RandomBest5Pct"))
          (AI|EQS|BindEventtoOnQueryFinished _returnvalue (delegate Custom|TeleportPointQueryFinished)))
        (:Is%20Not%20Valid
          _)))
    (CallFunction|RestartTeleportCooldown)
    (else
      (Networking|SwitchHasAuthority
        (:Authority
          (bind _returnvalue_2 (Utilities|Time|SetTimerbyEvent (EventDispatchers|CreateEvent :!event Custom|InitiateFireballsAbility) (ScalableFloat|GetValueatLevel (Variables|Hotfix|GetInitiateFireballsDelay_Hf))))
          (Variables|Default|SetShootFireballsTimerHandle _returnvalue_2)
          (Utilities|FlowControl|Delay (ScalableFloat|GetValueatLevel (Variables|Hotfix|GetTeleportGracePeriod_Hf)))
          (CallFunction|RestartTeleportCooldown))))))

(event Custom|TeleportPointQueryFinished
  (params (QueryInstance :type EQS%20Query%20Instance%20Object%20Reference) (QueryStatus :type EEnvQueryStatus%20Enum :default "Processing"))
  (bind _maskcombatzoffset_hf (Variables|Hotfix|GetMaskCombatZOffset_Hf))
  (bind _returnvalue (ScalableFloat|GetValueatLevel _maskcombatzoffset_hf))
  (bind _demonpawn (Variables|Default|GetDemonPawn))
  (bind (_spawnlocation_x _spawnlocation_y _spawnlocation_z) (Math|Vector|BreakVector (Variables|/FirePetalBosses/Pawn/BP_PlayerPawn_FirePetal_Boss.BP_PlayerPawn_FirePetal_Boss_C|GetSpawnLocation_X _demonpawn)))
  (bind _returnvalue_2 (+ _spawnlocation_z _returnvalue))
  ;[0.15 0.06635 0.02993 0.5] TeleportEvents
  (bind (_resultlocations _returnvalue_3) (AI|EQS|GetQueryResultsasLocations QueryInstance))
  (if _returnvalue_3
    (Variables|EyesControl|SetEyesIgnoreWeaponCollision true)
    (Variables|Default|SetInstantLookAt true)
    (GameplayCue|ExecuteGameplayCueOnActor(Burst) self (Variables|FX|GetGC_MaskTeleport) (Ability|GameplayCue|MakeGameplayCueParameters :NormalizedMagnitude 0.0 :RawMagnitude 0.0 :Location (Transformation|GetActorLocation) :Normal "0,0,0" :GameplayEffectLevel 1 :AbilityLevel 1 :bReplicateLocationWhenUsingMinimalRepProxy false))
    (bind (_output_x _output_y) (Math|Vector|BreakVector (Utilities|Array|Get(acopy) _resultlocations)))
    (Ability|ExecuteGameplayCueWithParamsReplicated (Variables|FX|GetGC_WarpTell) (Ability|GameplayCue|MakeGameplayCueParameters :NormalizedMagnitude 0.0 :RawMagnitude 0.0 :Location (Math|Vector|MakeVector _output_x _output_y _returnvalue_2) :Normal "0,0,0" :GameplayEffectLevel 1 :AbilityLevel 1 :bReplicateLocationWhenUsingMinimalRepProxy false))
    (Utilities|FlowControl|Delay 0.2)
    (Transformation|AttachActorToComponent (Variables|Default|GetDemonPawn) (Variables|Default|GetDemonHolderCombat) "None" "SnapToTarget" "SnapToTarget" "KeepWorld" false)
    (CallFunction|Teleport 0 (Math|Vector|MakeVector _output_x _output_y _returnvalue_2) false)
    (else
      (CallFunction|TeleportComplete false)
      (Development|PrintString "ERROR:: Puppeteer Mask:: Couldn't find any locations to teleport to."))))

(event Custom|TeleportComplete
  (params (ToSpawnCloud :type Boolean))
  ;[0.15 0.06635 0.02993 0.5] TeleportEvents
  (Variables|Default|SetTo%20Spawn%20Cloud ToSpawnCloud)
  (Utilities|FlowControl|Sequence
    (:then_0
      (Networking|SwitchHasAuthority
        (:Authority
          (GameplayCue|ExecuteGameplayCueOnActor(Burst) self (Variables|FX|GetGC_MaskAfterTeleport) (Ability|GameplayCue|MakeGameplayCueParameters :NormalizedMagnitude 0.0 :RawMagnitude 0.0 :Location (Transformation|GetActorLocation) :Normal "0,0,0" :GameplayEffectLevel 1 :AbilityLevel 1 :bReplicateLocationWhenUsingMinimalRepProxy false))
          (if ToSpawnCloud
            (Variables|EyesControl|SetEyesIgnoreWeaponCollision true)
            (CallFunction|ResetEyeDmg)
            (GameplayCue|RemoveGameplayCueOnActor(Looping) self (Variables|FX|GetGC_MaskAggressive_Loop) (Ability|GameplayCue|MakeGameplayCueParameters 0.0))
            (else
              (Utilities|FlowControl|Delay 0.2)
              (Variables|EyesControl|SetEyesIgnoreWeaponCollision false)
              (Variables|Default|SetInstantLookAt false)
              (bind _returnvalue (Utilities|Time|SetTimerbyEvent (EventDispatchers|CreateEvent :!event Custom|InitiateFireballsAbility) (ScalableFloat|GetValueatLevel (Variables|Hotfix|GetInitiateFireballsDelay_Hf))))
              (Variables|Default|SetShootFireballsTimerHandle _returnvalue)
              (Utilities|FlowControl|Delay (ScalableFloat|GetValueatLevel (Variables|Hotfix|GetTeleportGracePeriod_Hf)))
              (Class|BPFLAICommonFunctions|RemoveAllInstancesOfGameplayTag self "(TagName=\"PuppeteerBoss.Mask.Ability.Teleport.Activate\")")
              (CallFunction|RestartTeleportCooldown))))))
    (:then_1
      ; For Demon AnimBP
      (Variables|Default|SetIsMaskTeleport true)
      (Utilities|FlowControl|Delay 0.1)
      (Variables|Default|SetIsMaskTeleport false))))

(event EventOnDamagePlayEffects
  (params (Damage :type Float%20%28single-precision%29) (DamageTags :type Gameplay%20Tag%20Container%20Structure%20%28by%20ref%29) (Momentum :type Vector :default "0, 0, 0") (HitInfo :type Hit%20Result%20Structure%20%28by%20ref%29) (InstigatedBy :type Fort%20Pawn%20Object%20Reference) (DamageCauser :type Actor%20Object%20Reference) (EffectContext :type Gameplay%20Effect%20Context%20Handle%20Structure))
  (bind (_bblockinghit _binitialoverlap _time _distance _location _impactpoint _normal) (Collision|BreakHitResult HitInfo))
  (bind _returnvalue (Math|Conversions|RotationFromXVector _normal))
  ; RicochetImpactFX
  (if (not (Variables|EyesControl|GetIsEyeHit))
    (bind _ricochetimpactfx (Variables|Default|GetRicochetImpactFX)
      (:then
        ;[0.15 0.06635 0.02993 0.5] TeleportEvents
        (Transformation|SetWorldLocation _ricochetimpactfx _location false false)
        (Transformation|SetWorldRotation _ricochetimpactfx _returnvalue false false)
        (Components|Activation|Activate _ricochetimpactfx true))
      (:else
        (bind _returnvalue_2 (Effects|Components|ParticleSystem|SpawnFXAtLocation "/PuppeteerBoss/FX/NS_PuppeteerBoss_RicochetImpact.NS_PuppeteerBoss_RicochetImpact" _location _returnvalue))
        (bind _output_get (Variables|Default|SetRicochetImpactFX _returnvalue_2))))
    (Effects|Components|ParticleSystem|SetVectorParameter _ricochetimpactfx "ImpactFXDirection" _normal)))

(event Custom|InitiateFireballsAbility
  ;[0.4635 0.1315 0.06663 0.5] Fireballs
  (if (ScalableFloat|GetValueasBool (Variables|Hotfix|GetFireballAbilityEnabled_Hf))
    (Ability|SendGameplayEventtoActor self "(TagName=\"Ability.PuppeteerBoss.Mask.ShootFireballs\")")))

(event Custom|PlayEyeHitFX
  (bind _maskidlestringfx (Variables|Default|GetMaskIdleStringFX))
  ;[0.7045 0.1135 0.849 0.5] PlayEyeHitFX
  (Utilities|FlowControl|Sequence
    (:then_0
      (Networking|CanExecuteCosmeticEvents
        (:True
          (Utilities|FlowControl|DelayUntilNextTick)
          (Effects|Components|ParticleSystem|SetBoolParameter _maskidlestringfx "IsLeftEye" (Variables|EyesControl|GetLeftEyeHit))
          (Effects|Components|ParticleSystem|SetBoolParameter _maskidlestringfx "PlayHitFX" true)
          (Utilities|FlowControl|RetriggerableDelay 0.1)
          (Effects|Components|ParticleSystem|SetBoolParameter (Variables|Default|GetMaskIdleStringFX) "PlayHitFX" false))))
    (:then_1
      ; Added for AnimBP
      (Variables|EyesControl|SetIsEyeHit true)
      (Utilities|FlowControl|DelayUntilNextTick)
      (Variables|EyesControl|SetIsEyeHit false))))

(event Custom|EyeHit
  (params (EyeActor :type Actor%20Object%20Reference))
  ;[0.15 0.003125 0.003125 0.5] Damage/HealthHandling
  (CallFunction|PlayEyeHitFX)
  (if (== EyeActor (Variables|/Script/Engine.ChildActorComponent|GetChildActor (Variables|Default|GetEyeWeakpointL)))
    (Variables|EyesControl|SetLeftEyeHit true)
    (elif (== EyeActor (Variables|/Script/Engine.ChildActorComponent|GetChildActor (Variables|Default|GetEyeWeakpointR)))
      (Variables|EyesControl|SetLeftEyeHit false)))
  (if (Variables|FunctionControl|GetPauseHitReacts)
    _))

(event EventOnDamageServer
  (params (Damage :type Float%20%28single-precision%29) (DamageTags :type Gameplay%20Tag%20Container%20Structure%20%28by%20ref%29) (Momentum :type Vector :default "0, 0, 0") (HitInfo :type Hit%20Result%20Structure%20%28by%20ref%29) (InstigatedBy :type Controller%20Object%20Reference) (DamageCauser :type Actor%20Object%20Reference) (EffectContext :type Gameplay%20Effect%20Context%20Handle%20Structure))
  (bind _self self)
  (bind _returnvalue (Ability|GetFortAbilitySystemComponent _self))
  ;[0.15 0.003125 0.003125 0.5] Damage/HealthHandling
  (|CallParentFunction Damage DamageTags Momentum HitInfo InstigatedBy DamageCauser EffectContext)
  (bind _returnvalue_2 (Ability|GameplayEffect|AssignTagSetbyCallerMagnitude (GameplayEffects|MakeOutgoingSpec :self _returnvalue :GameplayEffectClass "/PuppeteerBoss/GameplayEffect/GE_Puppeteer_Demon_InstantHeal.GE_Puppeteer_Demon_InstantHeal_C" :Level 1.0) "(TagName=\"SetByCaller.Healing\")" Damage))
  (GameplayEffects|ApplyGameplayEffectSpecToTarget _returnvalue _returnvalue_2 _returnvalue))

(event Custom|DemonHealthChanged
  (bind _demonpawn (Variables|Default|GetDemonPawn))
  ;[0.15 0.003125 0.003125 0.5] Damage/HealthHandling
  (Utilities|IsValid _demonpawn
    (:Is%20Valid
      (Utilities|FlowControl|Sequence
        (:then_0
          (if (and (<= (FortPawn|GetHealth _demonpawn) 0) (GameplayTags|HasMatchingGameplayTag _demonpawn "(TagName=\"PuppeteerBoss.Demon.IsInPhase2\")"))
            (CallFunction|CleanUpFXOnDied)
            ;; for the AnimBP
            (Variables|Default|SetIsDeath true)
            (Utilities|FlowControl|Delay 2.0)
            (Actor|DestroyActor)))
        (:then_1
          (bind _audio (Variables|Default|GetAudio)
            (:then
              (Audio|Components|Audio|FadeOut _audio 1.0 0.0 "Linear" :!class /Script/Engine.AudioComponent))
            (:else)))))))

(event Custom|DemonPawnDied
  (params (DamagedActor :type Actor%20Object%20Reference) (Damage :type Float%20%28single-precision%29) (InstigatedBy :type Controller%20Object%20Reference) (DamageCauser :type Actor%20Object%20Reference) (HitLocation :type Vector) (FHitComponent :type Primitive%20Component%20Object%20Reference) (BoneName :type Name) (Momentum :type Vector))
  ;[0.15 0.003125 0.003125 0.5] Damage/HealthHandling
  (CallFunction|CleanUpFXOnDied)
  ;; for the AnimBP
  (Variables|Default|SetIsDeath true)
  (Utilities|FlowControl|Delay 2.0)
  (Actor|DestroyActor))

(event Custom|EyeBroken
  (params (EyeActor :type Actor%20Object%20Reference))
  ;[0.15 0.003125 0.003125 0.5] Damage/HealthHandling
  (Utilities|Time|ClearandInvalidateTimerbyHandle (Variables|Default|GetShootFireballsTimerHandle))
  (if (== EyeActor (Variables|/Script/Engine.ChildActorComponent|GetChildActor (Variables|Default|GetEyeWeakpointL)))
    (bind _output_get (Variables|EyesControl|SetLeftEyeBroken true))
    (Variables|EyesControl|SetLeftEyeIgnoreCollisions _output_get)
    (CallFunction|ToggleEyeGlow true)
    (GameplayCue|ExecuteGameplayCueOnActor(Burst) self (Variables|FX|GetGC_LeftEyeFlare) (Ability|GameplayCue|MakeGameplayCueParameters 0.0))
    (if (Variables|EyesControl|GetRightEyeBroken)
      (CallFunction|MaskStagger true)
      (else
        (CallFunction|MaskStagger true true)))
    (elif (== EyeActor (Variables|/Script/Engine.ChildActorComponent|GetChildActor (Variables|Default|GetEyeWeakpointR)))
      (bind _output_get_2 (Variables|EyesControl|SetRightEyeBroken true))
      (Variables|EyesControl|SetRightEyeIgnoreCollisions _output_get_2)
      (CallFunction|ToggleEyeGlow false)
      (GameplayCue|ExecuteGameplayCueOnActor(Burst) self (Variables|FX|GetGC_RightEyeFlare) (Ability|GameplayCue|MakeGameplayCueParameters 0.0))
      (if (Variables|EyesControl|GetLeftEyeBroken)
        (CallFunction|MaskStagger false)
        (else
          (CallFunction|MaskStagger false true)))))
  (Class|BPPlayerPawnPuppeteerDemon|MaskStagger (Variables|Default|GetDemonPawn))
  (if (not (GameplayTags|HasMatchingGameplayTag "(TagName=\"PuppeteerBoss.Mask.IsStaggered\")"))
    (Class|BPFLAICommonFunctions|AddUniqueGameplayTag self "(TagName=\"PuppeteerBoss.Mask.IsStaggered\")")
    (GameplayCue|RemoveGameplayCueOnActor(Looping) self (Variables|FX|GetGC_MaskAggressive_Loop) (Ability|GameplayCue|MakeGameplayCueParameters 0.0))
    (Variables|FunctionControl|SetPauseTeleport true)
    (Variables|FunctionControl|SetPauseLookAtTarget true)
    (CallFunction|WeaktheStringFXOnDBNO)))

(event Custom|CleanUpFXOnDied
  (bind _skeletalmeshcomponentbudgeted (Variables|Default|GetSkeletalMeshComponentBudgeted))
  ;[0 0 0 0.5] OnClientDie
  (Networking|CanExecuteCosmeticEvents
    (:True
      (Default|CallOnDead) ; UI Ref
      (Utilities|Time|ClearandInvalidateTimerbyHandle (Variables|Default|GetTickDemonPawnLocationTimerHandle))
      (Rendering|Material|SetMaterial _skeletalmeshcomponentbudgeted 0 "/PuppeteerBoss/Materials/MI_AI_PuppeteerBoss_Derez.MI_AI_PuppeteerBoss_Derez")
      (Rendering|Material|SetCustomPrimitiveDataFloat _skeletalmeshcomponentbudgeted 0 (Utilities|Time|GetGameTimeinSeconds))
      ;[0.15 0.003125 0.003125 0.5] Damage/HealthHandling
      (Effects|Components|ParticleSystem|SpawnFXAtLocation "/PuppeteerBoss/FX/NS_PuppeteerBoss_Mask_Teleport.NS_PuppeteerBoss_Mask_Teleport" (Transformation|GetActorLocation))
      (Effects|SpawnFXSystemAttached :FXSystemAsset "/PuppeteerBoss/FX/NS_PuppeteerBoss_Mask_DissolveAshes.NS_PuppeteerBoss_Mask_DissolveAshes" :AttachToComponent (Variables|Default|GetSkeletalMeshComponentBudgeted) :AttachPointName "root" :Scale "0.050000,0.050000,0.050000" :LocationType "KeepRelativeOffset" :bAutoDestroy true :PoolingMethod "None" :bAutoActivate true)
      (Utilities|FlowControl|Delay 1.8)
      (Components|Activation|Deactivate (Variables|Default|GetMaskIdleStringFX)))
    (:False
      (Utilities|FlowControl|Delay 1.8)
      (GameplayCue|ExecuteGameplayCueOnActor(Burst) self (Variables|Default|GetTag_MaskDeath) (Ability|GameplayCue|MakeGameplayCueParameters 0.0)))))

(event Custom|WeakTheStringFXOnDBNO
  (bind _maskidlestringfx (Variables|Default|GetMaskIdleStringFX))
  (Networking|CanExecuteCosmeticEvents
    (:True
      (Components|Activation|Deactivate _maskidlestringfx)
      (Utilities|FlowControl|DelayUntilNextTick)
      (Components|Activation|Activate _maskidlestringfx true)
      (Effects|Components|ParticleSystem|SetBoolParameter _maskidlestringfx "MaskDBNO" true)
      (Effects|Components|ParticleSystem|SetBoolParameter _maskidlestringfx "IsStringConnecting" false))))

(event Custom|StartMaskOscillation)

(event Custom|StopMaskOscillation)

(event Custom|ActivateEyeWeakPointControlBusMix)

(event Custom|StopUpDownOscillation
  ;[0.15 0.1075 0.04374 0.5] MaskLocalRotationAndLocationOffsets
  (bind (_direction _positionalpha) (Timeline|MaskUpDownOscillation_Timeline Stop :NewTime (Math|Random|RandomFloatinRange 0.0 2.0))
    (:Update
      (Variables|LocationControl|SetUpDownOffset (Math|Vector|Lerp(Vector) "0, 0, 0" (Math|Vector|MakeVector 0.0 0.0 (ScalableFloat|GetValueatLevel (Variables|Hotfix|GetOscillationOffset_Hf))) _positionalpha)))))

(event Custom|FireballShotAudioModulation)

(event Custom|StartUpDownOscillation
  ;[0.15 0.1075 0.04374 0.5] MaskLocalRotationAndLocationOffsets
  (Utilities|FlowControl|Sequence
    (:then_0
      (Timeline|MaskUpDownOscillation_Timeline SetNewTime))
    (:then_1
      (Timeline|MaskUpDownOscillation_Timeline Play))))

(event Custom|StopLeftRightOscillation
  ;[0.15 0.1075 0.04374 0.5] MaskLocalRotationAndLocationOffsets
  (bind (_direction _positionalpha) (Timeline|MaskLeftRightOscillation_Timeline Stop :NewTime (Math|Random|RandomFloatinRange 0.0 2.0))
    (:Update
      (Variables|LocationControl|SetLeftRightOffset (Math|Vector|Lerp(Vector) "0, 0, 0" (Math|Vector|MakeVector 0.0 (ScalableFloat|GetValueatLevel (Variables|Hotfix|GetOscillationOffset_Hf))) _positionalpha)))))

(event Custom|StartLeftRightOscillation
  ;[0.15 0.1075 0.04374 0.5] MaskLocalRotationAndLocationOffsets
  (Utilities|FlowControl|Sequence
    (:then_0
      (Timeline|MaskLeftRightOscillation_Timeline SetNewTime))
    (:then_1
      (Timeline|MaskLeftRightOscillation_Timeline Play))))

(event Custom|StopFrontBackOscillation
  ;[0.15 0.1075 0.04374 0.5] MaskLocalRotationAndLocationOffsets
  (bind (_direction _positionalpha) (Timeline|MaskFrontBackOscillation_Timeline Stop :NewTime (Math|Random|RandomFloatinRange 0.0 2.0))
    (:Update
      (Variables|LocationControl|SetFrontBackOffset (Math|Vector|Lerp(Vector) "0, 0, 0" (Math|Vector|MakeVector (ScalableFloat|GetValueatLevel (Variables|Hotfix|GetOscillationOffset_Hf))) _positionalpha)))))

(event Custom|StartFrontBackOscillation
  ;[0.15 0.1075 0.04374 0.5] MaskLocalRotationAndLocationOffsets
  (Utilities|FlowControl|Sequence
    (:then_0
      (Timeline|MaskFrontBackOscillation_Timeline SetNewTime))
    (:then_1
      (Timeline|MaskFrontBackOscillation_Timeline Play))))

(event Custom|StartMaskTwtichRotation)

(event Custom|MaskStagger
  (params (Left :type Boolean) (Partial :type Boolean))
  ;[0.15 0.1075 0.04374 0.5] MaskLocalRotationAndLocationOffsets
  (if Partial
    (Variables|RotationControl|SetLargeReactEndRotation "0, 0, 0")
    (else
      (Variables|RotationControl|SetLargeReactEndRotation "-65.000000,0.000000,0.000000")))
  (Variables|FunctionControl|SetPauseHitReacts true)
  (bind (_direction _rotationalpha) (Timeline|MaskLargeReact_Timeline PlayFromStart)
    (:EnableHitReacts
      (if (not Partial)
        (Variables|FunctionControl|SetPauseHitReacts false)))))

(event Custom|ResetLocalRotation
  ;[0.15 0.1075 0.04374 0.5] MaskLocalRotationAndLocationOffsets
  (CallFunction|ResetEyeGlowFX))

(event Custom|ResetEyeGlowFX
  (bind _maskidlestringfx (Variables|Default|GetMaskIdleStringFX))
  (bind _maskidlestringfx_2 (Variables|Default|GetMaskIdleStringFX))
  ;[0.15 0.1075 0.04374 0.5] MaskLocalRotationAndLocationOffsets
  (Networking|CanExecuteCosmeticEvents
    (:True
      (Effects|Components|ParticleSystem|SetBoolParameter _maskidlestringfx "HideLeftEyeFX" false)
      (Effects|Components|ParticleSystem|SetBoolParameter _maskidlestringfx "HideRightEyeFX" false)
      (CallFunction|UpdateStringFXonMask (Transformation|GetSocketLocation (Variables|/Script/Engine.Character|GetMesh (Variables|Default|GetDemonPawn)) "hand_l") (Transformation|GetSocketLocation (Variables|/Script/Engine.Character|GetMesh (Variables|Default|GetDemonPawn)) "hand_r"))
      (Effects|Components|ParticleSystem|SetBoolParameter _maskidlestringfx_2 "MaskDBNO" false)
      (Effects|Components|ParticleSystem|SetBoolParameter _maskidlestringfx_2 "IsStringConnecting" true)
      (Effects|Components|ParticleSystem|SetBoolParameter (Variables|Default|GetMaskIdleStringFX) "EnableWeakpoints" (and (not (Variables|Default|GetTo%20Spawn%20Cloud)) (Variables|/PuppeteerBoss/Pawn/BP_PlayerPawn_Puppeteer_Demon.BP_PlayerPawn_Puppeteer_Demon_C|GetIsStartPhase2 (Variables|Default|GetDemonPawn)))))))

(event EventEndPlay
  (params (EndPlayReason :type EEndPlayReason%20Enum :default "Destroyed"))
  ; EndPlay
  (|CallParentFunction EndPlayReason)
  (Utilities|Time|ClearandInvalidateTimerbyHandle (Variables|Default|GetLookAtTargetTimerHandle)))

(event Custom|ActivateAudioEffect
  (params (FortPawn :type Fort%20Pawn%20Object%20Reference) (bInIsDBNO :type Boolean))
  (bind _maskidlestringfx (Variables|Default|GetMaskIdleStringFX))
  (bind _maskidlestringfx_2 (Variables|Default|GetMaskIdleStringFX))
  (bind _audio (Variables|Default|GetAudio))
  (Utilities|FlowControl|Sequence
    (:then_0
      (Components|Activation|Activate _audio false)
      (Utilities|FlowControl|Delay 3.0)
      (Components|Activation|Deactivate _audio))
    (:then_1
      ; Start Phase 2
      (if bInIsDBNO
        (Components|Activation|Deactivate _maskidlestringfx)
        (Utilities|FlowControl|DelayUntilNextTick)
        (Components|Activation|Activate _maskidlestringfx true)
        (Effects|Components|ParticleSystem|SetBoolParameter _maskidlestringfx "MaskDBNO" true)
        (Effects|Components|ParticleSystem|SetBoolParameter _maskidlestringfx "IsStringConnecting" false)
        (Utilities|FlowControl|Delay 3.0)
        (Effects|Components|ParticleSystem|SetBoolParameter _maskidlestringfx "EnableWeakpoints" true)
        (Components|Activation|Deactivate _maskidlestringfx_2)
        (Utilities|FlowControl|DelayUntilNextTick)
        (Components|Activation|Activate _maskidlestringfx_2 true)
        (Effects|Components|ParticleSystem|SetBoolParameter _maskidlestringfx_2 "MaskDBNO" false)
        (Effects|Components|ParticleSystem|SetBoolParameter _maskidlestringfx_2 "IsStringConnecting" true)))
    (:then_2
      (Utilities|FlowControl|DoOnce false
        (:Completed
          (Utilities|FlowControl|Delay 1.0)
          (bind _returnvalue (Effects|Components|ParticleSystem|SpawnFXAtLocation "/PuppeteerBoss/FX/NS_PuppeteerBoss_MaskAppear_Small_Wramup.NS_PuppeteerBoss_MaskAppear_Small_Wramup" (Transformation|GetWorldLocation (Variables|Transformation|GetRootComponent (Variables|Default|GetDemonPawn)))))
          (Variables|FX|SetWramUpFX _returnvalue)
          (Utilities|FlowControl|Delay 2.0)
          (Effects|Components|ParticleSystem|SpawnFXAtLocation "/PuppeteerBoss/FX/NS_PuppeteerBoss_MaskAppear_Small.NS_PuppeteerBoss_MaskAppear_Small" (+ (Transformation|GetWorldLocation (Variables|Transformation|GetRootComponent (Variables|Default|GetDemonPawn))) "0.000000,0.000000,-80.000000"))
          (Components|Activation|Deactivate (Variables|FX|GetWramUpFX)))))))

(fn FindAndInterpTargetLookAtRotation
  (params (NoInterp :type Boolean) (DeltaTimeToUse :type Float%20%28double-precision%29))
  (bind _returnvalue (Transformation|GetActorRotation))
  (bind _returnvalue_2 (Transformation|GetActorLocation))
  (bind _instantlookat (Variables|Default|GetInstantLookAt))
  (bind _rotationinterpspeed_hf (Variables|Hotfix|GetRotationInterpSpeed_Hf))
  (bind _returnvalue_3 (ScalableFloat|GetValueatLevel _rotationinterpspeed_hf))
  (bind _returnvalue_4 (Math|Float|SelectFloat 0.0 _returnvalue_3 _instantlookat))
  (bind _playertarget (Variables|Default|GetPlayerTarget))
  (bind _demonpawn (Variables|Default|GetDemonPawn))
  (if (not (Variables|FunctionControl|GetForceLookAtDemon))
    (Utilities|IsValid _playertarget
      (:Is%20Valid
        (return (Math|Interpolation|RInterpTo _returnvalue (Math|Rotator|FindLookatRotation _returnvalue_2 (Transformation|GetActorLocation _playertarget)) DeltaTimeToUse _returnvalue_4)))
      (:Is%20Not%20Valid
        _)))
  (Utilities|IsValid _demonpawn
    (:Is%20Valid
      (return (Math|Interpolation|RInterpTo _returnvalue (Math|Rotator|FindLookatRotation _returnvalue_2 (Transformation|GetActorLocation _demonpawn)) DeltaTimeToUse _returnvalue_4)))
    (:Is%20Not%20Valid
      (return _returnvalue))))

(fn GetDataForDemonAnimBP
  (return (Variables|Default|GetPlayerTarget) (Variables|Default|GetIsMaskTeleport)))

(fn GetDataForMaskAnimBP
  (return (Variables|Default|GetDemonPawn) (Variables|Default|GetPlayerTarget) (Variables|EyesControl|GetIsEyeHit) (Variables|EyesControl|GetLeftEyeHit) (Variables|EyesControl|GetLeftEyeBroken) (Variables|EyesControl|GetRightEyeBroken) (Variables|Default|GetIsDeath)))

(fn OnRep_bFireballShot
  (Networking|SwitchHasAuthority
    (:Remote
      (CallFunction|FireballShotAudioModulation))))

(fn OnRep_bShouldPlayAudioImpact)

(fn OnRep_bShouldPlayAwakeSound
  (Networking|SwitchHasAuthority
    (:Remote
      (if (Variables|Audio|GetbShouldPlayAwakeSound)
        (Audio|SpawnSoundAttached (Variables|Audio|GetAwakeSound) (Variables|Default|GetAudio))))))

(fn OnRep_bShouldPlayIdleSound
  (bind _audio (Variables|Default|GetAudio))
  (Networking|SwitchHasAuthority
    (:Remote
      (if (Variables|Audio|GetbShouldPlayIdleSound)
        (Audio|Components|Audio|FadeIn _audio 1.0 1.0 0.0 "Linear" :!class /Script/Engine.AudioComponent)
        (else
          (Audio|Components|Audio|FadeOut _audio 1.0 0.0 "Linear" :!class /Script/Engine.AudioComponent))))))

(fn OnRep_DemonPawn
  (Networking|SwitchHasAuthority
    (:Remote
      (CallFunction|DemonPawnSet (Variables|Default|GetDemonPawn)))))

(fn OnRep_EyesIgnoreWeaponCollision
  (bind _eyesignoreweaponcollision (Variables|EyesControl|GetEyesIgnoreWeaponCollision))
  (Variables|EyesControl|SetLeftEyeIgnoreCollisions _eyesignoreweaponcollision)
  (Variables|EyesControl|SetRightEyeIgnoreCollisions _eyesignoreweaponcollision))

(fn OnRep_FrontBackOffset)

(fn OnRep_HitRotation)

(fn OnRep_IsEyeHit)

(fn OnRep_LeftEyeBroken
  (Networking|SwitchHasAuthority
    (:Remote
      (CallFunction|ActivateEyeWeakPointControlBusMix))))

(fn OnRep_LeftEyeIgnoreCollisions
  (bind _lefteyeactor (Variables|EyesControl|GetLeftEyeActor))
  (bind _lefteyeignorecollisions (Variables|EyesControl|GetLeftEyeIgnoreCollisions))
  (Networking|SwitchHasAuthority
    (:Authority
      (Utilities|IsValid _lefteyeactor
        (:Is%20Valid
          (Variables|/PuppeteerBoss/Mask/BGA_PuppeteerMaskEyeWeakpoint.BGA_PuppeteerMaskEyeWeakpoint_C|SetIgnoreCollisions _lefteyeignorecollisions _lefteyeactor)
          (if (not _lefteyeignorecollisions)
            _))))))

(fn OnRep_LeftRightOffset)

(fn OnRep_LocalRotation)

(fn OnRep_MaskIsVisible
  (Networking|SwitchHasAuthority
    (:Remote
      (Rendering|SetVisibility (Variables|Default|GetSkeletalMeshComponentBudgeted) (Variables|Default|GetMaskIsVisible) false))))

(fn OnRep_MaskLocation
  (Transformation|SetActorLocation (Variables|LocationControl|GetMaskLocation) false true)
  (Default|CallTeleportCall))

(fn OnRep_MaskRotation
  (Transformation|SetActorRotation (Variables|RotationControl|GetMaskRotation)))

(fn OnRep_RightEyeBroken
  (Networking|SwitchHasAuthority
    (:Remote
      (CallFunction|ActivateEyeWeakPointControlBusMix))))

(fn OnRep_RightEyeIgnoreCollisions
  (bind _righteyeactor (Variables|EyesControl|GetRightEyeActor))
  (bind _righteyeignorecollisions (Variables|EyesControl|GetRightEyeIgnoreCollisions))
  (Networking|SwitchHasAuthority
    (:Authority
      (Utilities|IsValid _righteyeactor
        (:Is%20Valid
          (Variables|/PuppeteerBoss/Mask/BGA_PuppeteerMaskEyeWeakpoint.BGA_PuppeteerMaskEyeWeakpoint_C|SetIgnoreCollisions _righteyeignorecollisions _righteyeactor)
          (if (not _righteyeignorecollisions)
            _))))))

(fn OnRep_UpDownOffset)

(fn ResetEyeDmg
  (bind _lefteyeactor (Variables|EyesControl|GetLeftEyeActor))
  (bind _righteyeactor (Variables|EyesControl|GetRightEyeActor))
  (bind _demonpawn (Variables|Default|GetDemonPawn))
  (Utilities|FlowControl|Sequence
    (:then_0
      (Utilities|IsValid _lefteyeactor
        (:Is%20Valid
          (Variables|/PuppeteerBoss/Mask/BGA_PuppeteerMaskEyeWeakpoint.BGA_PuppeteerMaskEyeWeakpoint_C|SetCurrentDmg 0.0 _lefteyeactor)
          (Variables|EyesControl|SetLeftEyeBroken false))))
    (:then_1
      (Utilities|IsValid _righteyeactor
        (:Is%20Valid
          (Variables|/PuppeteerBoss/Mask/BGA_PuppeteerMaskEyeWeakpoint.BGA_PuppeteerMaskEyeWeakpoint_C|SetCurrentDmg 0.0 _righteyeactor)
          (Variables|EyesControl|SetRightEyeBroken false))))
    (:then_2
      (Variables|FunctionControl|SetPauseHitReacts false)
      (Variables|FunctionControl|SetPauseLookAtTarget false)
      (CallFunction|ResetLocalRotation)
      (CallFunction|ToggleEyeGlow false true)
      (Class|BPFLAICommonFunctions|RemoveAllInstancesOfGameplayTag self "(TagName=\"PuppeteerBoss.Mask.IsStaggered\")")
      (Utilities|IsValid _demonpawn
        (:Is%20Valid
          (if (not (GameplayTags|HasAnyMatchingGameplayTags _demonpawn (GameplayTags|MakeLiteralGameplayTagContainer "(GameplayTags=((TagName=\"FirePetalBosses.Demons.IsTeleportingToSpawnCloud\"),(TagName=\"FirePetalBosses.Demons.IsinSpawnCloud\")))")))
            (Variables|FunctionControl|SetPauseTeleport false)))))))

(fn SetStringFXTarget
  (bind _demonpawnclient (Variables|Default|GetDemonPawnClient))
  (bind _mesh (Variables|/Script/Engine.Character|GetMesh _demonpawnclient))
  (bind _maskidlestringfx (Variables|Default|GetMaskIdleStringFX))
  (Effects|Components|ParticleSystem|SetVectorParameter _maskidlestringfx "StringTargetL" (Transformation|GetSocketLocation _mesh "hand_l"))
  (Effects|Components|ParticleSystem|SetVectorParameter _maskidlestringfx "StringTargetR" (Transformation|GetSocketLocation _mesh "hand_r"))
  (return))

(fn Teleport
  (params (Location :type Vector) (ToSpawnCloud :type Boolean))
  (bind _returnvalue (Transformation|Teleport Location (Transformation|GetActorRotation)))
  (if _returnvalue
    (Variables|LocationControl|SetMaskLocation (Transformation|GetActorLocation))
    (Variables|RotationControl|SetMaskRotation (select (Variables|GetToSpawnCloud) (Transformation|GetActorRotation (Variables|Default|GetDemonPawn)) (Transformation|GetActorRotation)))
    (bind _outhits (Collision|MultiCapsuleTraceByChannel (Transformation|GetWorldLocation (Variables|Default|GetPropCollisionCapsuleDebug)) (+ (Transformation|GetWorldLocation (Variables|Default|GetPropCollisionCapsuleDebug)) "0.000000,0.000000,1.000000") (ScalableFloat|GetValueatLevel (Variables|Hotfix|GetPropCollisionCapsuleRadius_Hf)) (ScalableFloat|GetValueatLevel (Variables|Hotfix|GetPropCollisionCapsuleHalfHeight_Hf)) "TraceTypeQuery10" false (Utilities|Array|MakeArray (Variables|EyesControl|GetLeftEyeActor) (Variables|EyesControl|GetRightEyeActor))))
    (for _array_element _outhits
      (Utilities|IsValid (.HitActor (Collision|BreakHitResult _array_element))
        (:Is%20Valid
          (Game|Damage|ApplyDamage (.HitActor (Collision|BreakHitResult _array_element)) 1e+04 (Pawn|GetController (Variables|Default|GetDemonPawn)) self "/Engine/EngineDamageTypes/DmgTypeBP_Environmental.DmgTypeBP_Environmental_C"))))
    (CallFunction|TeleportComplete (Variables|GetToSpawnCloud))
    (else
      (Development|PrintString "ERROR:: Puppeteer Mask:: Tried to teleport but it failed."))))

(fn UpdateMeshWeaponCollision
  (params (Ignore :type Boolean) (SkeletalMesh :type Skeletal%20Mesh%20Component%20Object%20Reference))
  (bind _skeletalmesh (Variables|GetSkeletalMesh))
  (Collision|SetCollisionResponsetoChannel _skeletalmesh "ECC_GameTraceChannel7" (select (Variables|GetIgnore) "ECR_Ignore" "ECR_Block"))
  (Collision|SetCollisionResponsetoChannel _skeletalmesh "ECC_GameTraceChannel14" (select (Variables|GetIgnore) "ECR_Ignore" "ECR_Overlap"))
  (Collision|SetCollisionResponsetoChannel _skeletalmesh "ECC_WorldStatic" (select (Variables|GetIgnore) "ECR_Ignore" "ECR_Overlap"))
  (Collision|SetCollisionResponsetoChannel _skeletalmesh "ECC_WorldDynamic" (select (Variables|GetIgnore) "ECR_Ignore" "ECR_Overlap"))
  (Collision|SetCollisionResponsetoChannel _skeletalmesh "ECC_Pawn" (select (Variables|GetIgnore) "ECR_Ignore" "ECR_Overlap"))
  (Collision|SetCollisionResponsetoChannel _skeletalmesh "ECC_PhysicsBody" (select (Variables|GetIgnore) "ECR_Ignore" "ECR_Overlap"))
  (Collision|SetCollisionResponsetoChannel _skeletalmesh "ECC_Vehicle" (select (Variables|GetIgnore) "ECR_Ignore" "ECR_Overlap"))
  (Collision|SetCollisionResponsetoChannel _skeletalmesh "ECC_Destructible" (select (Variables|GetIgnore) "ECR_Ignore" "ECR_Overlap"))
  (Collision|SetCollisionResponsetoChannel _skeletalmesh "ECC_GameTraceChannel4" (select (Variables|GetIgnore) "ECR_Ignore" "ECR_Overlap"))
  (Collision|SetCollisionResponsetoChannel _skeletalmesh "ECC_GameTraceChannel8" (select (Variables|GetIgnore) "ECR_Ignore" "ECR_Block"))
  (Collision|SetCollisionResponsetoChannel _skeletalmesh "ECC_GameTraceChannel9" (select (Variables|GetIgnore) "ECR_Ignore" "ECR_Block"))
  (Collision|SetCollisionResponsetoChannel _skeletalmesh "ECC_GameTraceChannel12" (select (Variables|GetIgnore) "ECR_Ignore" "ECR_Block")))

(fn ConstructionScript)

(fn ValidateTeleportAreaBox
  (Utilities|IsValid (Variables|Default|GetTeleportAreaBox)
    (:Is%20Valid
      (bind _output_get (Variables|FunctionControl|SetTeleportAreaBoxValidated true))
      (return _output_get))
    (:Is%20Not%20Valid
      (bind _outactors (Actor|GetAllActorsOfClass "/PuppeteerBoss/Box/Puppeteer_MaskTeleportArea_Box.Puppeteer_MaskTeleportArea_Box_C"))
      (Utilities|IsValid (Utilities|Array|Get(acopy) _outactors)
        (:Is%20Valid
          (Variables|Default|SetTeleportAreaBox (Utilities|Array|Get(acopy) _outactors))
          (bind _output_get_2 (Variables|FunctionControl|SetTeleportAreaBoxValidated true))
          (return _output_get_2))
        (:Is%20Not%20Valid
          (bind _output_get_3 (Variables|FunctionControl|SetTeleportAreaBoxValidated false))
          (Development|PrintString "ERROR:: Puppeteer Mask:: Couldn't validate teleport area box.")
          (return _output_get_3))))))
