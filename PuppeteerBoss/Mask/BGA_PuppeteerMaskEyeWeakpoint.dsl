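; Damageable eye weakpoint on the Puppeteer boss mask that reports hits and breakage to its parent mask actor.
; On the server, it heals back all damage while accumulating a total, calling EyeHit or EyeBroken at most once per frame against a threshold scaled by playlist team size.
; Replicated IgnoreCollisions switches collision responses and invulnerability and gates auto-fire targeting.
; Remote clients viewing the damaging player hear a hitmarker sound.
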(event EventBeginPlay
  (Networking|SwitchHasAuthority
    (:Authority
      (bind (_gamestate _playlist) (Class|FortAsyncActionCurrentPlaylistReady|PlaylistDataReadyAsync)
        (:OnReady
          (bind _asbga_puppeteer_mask (Utilities|Casting|CastToBGA_PuppeteerMask (Actor|GetParentActor)))
          (Variables|Default|SetMaskActor _asbga_puppeteer_mask)
          (Health|SetMaxHealthDontCreateASC 1e+04)
          (bind _output_get (Variables|Default|SetMaxTeamSize (Math|Integer|Clamp(Integer) (Variables|/Script/FortniteGame.FortPlaylist|GetMaxTeamSize _playlist) 1 4)))
          (Variables|Default|SetEyeDmgThreshold (ScalableFloat|GetValueatLevel (Variables|Default|GetEyeDmgThreshold_Hf) (Math|Conversions|ToFloat(Integer) _output_get))))))))

(event EventOnDamageServer
  (params (Damage :type Float%20%28single-precision%29) (DamageTags :type Gameplay%20Tag%20Container%20Structure%20%28by%20ref%29) (Momentum :type Vector :default "0, 0, 0") (HitInfo :type Hit%20Result%20Structure%20%28by%20ref%29) (InstigatedBy :type Controller%20Object%20Reference) (DamageCauser :type Actor%20Object%20Reference) (EffectContext :type Gameplay%20Effect%20Context%20Handle%20Structure))
  (bind _self self)
  (bind _returnvalue (Ability|GetFortAbilitySystemComponent _self))
  (bind _returnvalue_2 (Utilities|GetFrameCount))
  (bind _self_2 self)
  (bind _maskactor (Variables|Default|GetMaskActor))
  (|CallParentFunction Damage DamageTags Momentum HitInfo InstigatedBy DamageCauser EffectContext)
  (bind _returnvalue_3 (Ability|GameplayEffect|AssignTagSetbyCallerMagnitude (GameplayEffects|MakeOutgoingSpec :self _returnvalue :GameplayEffectClass "/PuppeteerBoss/GameplayEffect/GE_Puppeteer_Demon_InstantHeal.GE_Puppeteer_Demon_InstantHeal_C" :Level 1.0) "(TagName=\"SetByCaller.Healing\")" Damage))
  (GameplayEffects|ApplyGameplayEffectSpecToTarget _returnvalue _returnvalue_3 _returnvalue)
  (bind _output_get (Variables|Default|SetCurrentDmg (+ Damage (Variables|Default|GetCurrentDmg))))
  (if (not (== _returnvalue_2 (Variables|Default|GetLastFrameCount)))
    (Variables|Default|SetLastFrameCount _returnvalue_2)
    (if (>= _output_get (Variables|Default|GetEyeDmgThreshold))
      (Class|BGAPuppeteerMask|EyeBroken _maskactor _self_2)
      (else
        (Class|BGAPuppeteerMask|EyeHit _maskactor _self_2)))))

(event EventOnDamagePlayEffects
  (params (Damage :type Float%20%28single-precision%29) (DamageTags :type Gameplay%20Tag%20Container%20Structure%20%28by%20ref%29) (Momentum :type Vector :default "0, 0, 0") (HitInfo :type Hit%20Result%20Structure%20%28by%20ref%29) (InstigatedBy :type Fort%20Pawn%20Object%20Reference) (DamageCauser :type Actor%20Object%20Reference) (EffectContext :type Gameplay%20Effect%20Context%20Handle%20Structure))
  (bind _asfort_player_pawn_athena (Utilities|Casting|CastToFortPlayerPawnAthena InstigatedBy))
  (bind _output_get (Variables|Default|SetDamagingInstagatorPawn _asfort_player_pawn_athena))
  (if (Pawn|IsLocallyViewed _output_get)
    (Networking|SwitchHasAuthority
      (:Remote
        (Audio|SpawnSound2D (Variables|Default|GetHitmarkerSound))))))

(fn GetTargetOwner
  (return 0))

(fn IsValidAutoFireTarget
  (params (Attacker :type Actor%20Object%20Reference))
  (return (not (Variables|Default|GetIgnoreCollisions))))

(fn OnRep_IgnoreCollisions
  (Default|UpdateMeshWeaponCollision (Variables|Default|GetIgnoreCollisions) (Variables|Default|GetWeakpointMesh)))

(fn ShouldAlliesAutoFireOnDeployedTarget
  (return false))

(fn ShouldAutoFireOnSelfDeployedTarget
  (return false))

(fn UpdateMeshWeaponCollision
  (params (Ignore :type Boolean) (StaticMesh :type Static%20Mesh%20Component%20Object%20Reference))
  (bind _staticmesh (Variables|GetStaticMesh))
  (Collision|SetCollisionResponsetoChannel _staticmesh "ECC_GameTraceChannel7" (select (Variables|GetIgnore) "ECR_Ignore" "ECR_Block"))
  (Collision|SetCollisionResponsetoChannel _staticmesh "ECC_GameTraceChannel14" (select (Variables|GetIgnore) "ECR_Ignore" "ECR_Overlap"))
  (Collision|SetCollisionResponsetoChannel _staticmesh "ECC_WorldStatic" (select (Variables|GetIgnore) "ECR_Ignore" "ECR_Overlap"))
  (Collision|SetCollisionResponsetoChannel _staticmesh "ECC_WorldDynamic" (select (Variables|GetIgnore) "ECR_Ignore" "ECR_Overlap"))
  (Collision|SetCollisionResponsetoChannel _staticmesh "ECC_Pawn" (select (Variables|GetIgnore) "ECR_Ignore" "ECR_Overlap"))
  (Collision|SetCollisionResponsetoChannel _staticmesh "ECC_PhysicsBody" (select (Variables|GetIgnore) "ECR_Ignore" "ECR_Overlap"))
  (Collision|SetCollisionResponsetoChannel _staticmesh "ECC_Vehicle" (select (Variables|GetIgnore) "ECR_Ignore" "ECR_Overlap"))
  (Collision|SetCollisionResponsetoChannel _staticmesh "ECC_Destructible" (select (Variables|GetIgnore) "ECR_Ignore" "ECR_Overlap"))
  (Collision|SetCollisionResponsetoChannel _staticmesh "ECC_GameTraceChannel4" (select (Variables|GetIgnore) "ECR_Ignore" "ECR_Overlap"))
  (Collision|SetCollisionResponsetoChannel _staticmesh "ECC_GameTraceChannel8" (select (Variables|GetIgnore) "ECR_Ignore" "ECR_Block"))
  (Collision|SetCollisionResponsetoChannel _staticmesh "ECC_GameTraceChannel9" (select (Variables|GetIgnore) "ECR_Ignore" "ECR_Block"))
  (Collision|SetCollisionResponsetoChannel _staticmesh "ECC_GameTraceChannel12" (select (Variables|GetIgnore) "ECR_Ignore" "ECR_Block"))
  (Variables|BuildingActor|SetbIsInvulnerable (Variables|GetIgnore)))

(fn ConstructionScript)
