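; Fire petal projectile Blueprint for the Puppeteer boss's mask fire attack, extending a parent class that receives all its events.
; On BeginPlay it adds a looping fireball GameplayCue to itself, read from the GC_MaskFireBall_Loop variable.
; On Hit it calls the parent handler, then removes that looping cue.
; The child adds only this cue lifecycle; movement and impact logic are not defined here.
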(event EventBeginPlay
  (|CallParentFunction)
  (GameplayCue|AddGameplayCueOnActor(Looping) self (Variables|Default|GetGC_MaskFireBall_Loop) (Ability|GameplayCue|MakeGameplayCueParameters 0.0)))

(event Collision|EventHit
  (params (MyComp :type Primitive%20Component%20Object%20Reference) (Other :type Actor%20Object%20Reference) (OtherComp :type Primitive%20Component%20Object%20Reference) (bSelfMoved :type Boolean) (HitLocation :type Vector :default "0, 0, 0") (HitNormal :type Vector :default "0, 0, 0") (NormalImpulse :type Vector :default "0, 0, 0") (Hit :type Hit%20Result%20Structure%20%28by%20ref%29))
  (|CallParentFunction MyComp Other OtherComp bSelfMoved HitLocation HitNormal NormalImpulse Hit)
  (GameplayCue|RemoveGameplayCueOnActor(Looping) self (Variables|Default|GetGC_MaskFireBall_Loop) (Ability|GameplayCue|MakeGameplayCueParameters 0.0)))

(fn ConstructionScript
  (|CallParentFunction))
